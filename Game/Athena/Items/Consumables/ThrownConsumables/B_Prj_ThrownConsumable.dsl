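; Base thrown-consumable projectile blueprint that layers cosmetic feedback onto parent projectile behavior.
; Construction and simulation reset attach rotating movement to the mesh, stopping detaches it, BeginPlay sets the fuse sound, and bounces play collision audio with sound indicators.
; On explosion, cosmetic-capable instances spawn particles, shake the camera, play close or distant sound by local instigator, show an indicator, play force feedback, hide the mesh and deactivate fuse particles.
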(event EventBeginPlay
  ;[1 1 1] BeginPlay - Set
  (|CallParentFunction)
  (Networking|CanExecuteCosmeticEvents
    (:True
      (Audio|Components|Audio|SetSound (Variables|Default|GetGrenadeFuse_AudioComponent) (Variables|Cosmetic|GetCue_GrenadeFuseSound)))))

(event EventOnBounce
  (params (Hit :type Hit%20Result%20Structure%20%28by%20ref%29))
  (bind (_bblockinghit _binitialoverlap _time _distance _location _impactpoint _normal _impactnormal) (Collision|BreakHitResult Hit))
  ;[1 1 1] OnBounce - Play bounce audio
  (|CallParentFunction Hit)
  (Utilities|IsValid (Variables|Cosmetic|GetCue_Bounce)
    (:Is%20Valid
      (bind _outresult (Audio|OnCollision (Variables|Default|GetFortCollisionAudio) _location (* _impactnormal (Variables|/Script/Engine.MovementComponent|GetVelocity (Variables|Movement|GetProjectileMovementComponent)))))
      (if _outresult
        (Fortnite|ShowSoundIndicatorLocalforTeam :Location _location :MaxAudibleDistance (Variables|/Script/Engine.SoundBase|GetMaxDistance (Variables|Cosmetic|GetCue_Bounce)) :IndicatorType "Gunshot" :Team (Team|GetActorTeam self) :Affiliations (Utilities|Array|MakeArray "Hostile" "Neutral") :Instigator self :Tint "(R=0.000000,G=0.000000,B=0.000000,A=0.000000)" :AttenuationMaxRangeMultiplier 1.0 :AttenuationIntensityMultiplier 1.0)))))

(event EventOnStop
  (params (Hit :type Hit%20Result%20Structure%20%28by%20ref%29))
  ;[0.7116 0.838 1] OnStop - Stop the rotation
  (|CallParentFunction Hit)
  (Components|Movement|SetUpdatedComponent (Variables|Default|GetRotatingMovement) 0))

(event EventOnExploded
  (params (HitActors :type Array%20of%20Actor%20Object%20References) (HitResults :type Array%20of%20Hit%20Result%20Structures))
  (bind _returnvalue (Transformation|GetActorLocation))
  (bind _cue_distantsound (Variables|Cosmetic|GetCue_DistantSound))
  ;[1 1 1] OnExploded - Emitter, Sound, Camera Shake
  (|CallParentFunction HitActors HitResults)
  (Networking|CanExecuteCosmeticEvents
    (:True
      (bind _returnvalue_2 (Effects|SpawnEmitteratLocation (Variables|Cosmetic|GetP_Explosion) (Transformation|GetActorLocation) (Variables|Default|GetExplosionRotation)))
      (Rendering|SetRenderCustomDepth _returnvalue_2 true)
      (Camera|PlayWorldCameraShake (Variables|Cosmetic|GetExplosionCameraShake) (Transformation|GetActorLocation) 200 800)
      (if (Pawn|IsLocallyControlled (Game|GetInstigator))
        (Audio|PlaySoundatLocation (Variables|Cosmetic|GetCue_CloseSound) _returnvalue)
        (Audio|Components|Audio|FadeOut (Variables|Default|GetGrenadeFuse_AudioComponent) 0.2 0.0 "Linear" :!class /Script/Engine.AudioComponent) ; Fade Out Fuse Sound
        (else
          (Audio|PlaySoundatLocation (Variables|Cosmetic|GetCue_DistantSound) _returnvalue)))
      (Utilities|IsValid _cue_distantsound
        (:Is%20Valid
          (Fortnite|ShowSoundIndicatorLocalforTeam :Location (Transformation|GetActorLocation) :MaxAudibleDistance (Variables|/Script/Engine.SoundBase|GetMaxDistance _cue_distantsound) :IndicatorType "Gunshot" :Team (Team|GetActorTeam self) :Affiliations (Utilities|Array|MakeArray "Hostile" "Neutral") :Tint "(R=0.000000,G=0.000000,B=0.000000,A=0.000000)" :AttenuationMaxRangeMultiplier 1.0 :AttenuationIntensityMultiplier 1.0))
        (:Is%20Not%20Valid
          _))
      (Fortnite|PlayLocalForceFeedbackatLocationMulti (Variables|Cosmetic|GetExplosionForceFeedbackNear) (Variables|Cosmetic|GetExplosionForceFeedbackFar) (Transformation|GetActorLocation) (Variables|Default|GetExplosionRadius) (* (Variables|Default|GetExplosionRadius) 3) "Explosion")
      (Development|SetHiddeninGame (Variables|Default|GetMesh) true false)
      (Components|Activation|Deactivate (Variables|Default|GetFuse_Particle)))))

(event EventResetSimulatedComponents
  ;[1 1 1] ResetSimulatedComponents - Call to parent function, then resimulate cosmetic components
  (|CallParentFunction)
  (Networking|CanExecuteCosmeticEvents
    (:True
      (Components|Movement|SetUpdatedComponent (Variables|Default|GetRotatingMovement) (Variables|Default|GetMesh)))))

(fn ConstructionScript
  (|CallParentFunction)
  (Networking|CanExecuteCosmeticEvents
    (:True
      (Components|Movement|SetUpdatedComponent (Variables|Default|GetRotatingMovement) (Variables|Default|GetMesh))
      (Audio|SetCollisionSoundatIndex (Variables|Default|GetFortCollisionAudio) 0 (Variables|Cosmetic|GetCue_Bounce)))))
