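; Parent projectile for thrown consumables, with a fuse timer calling an empty FuseEnded hook and a delayed explosion on any damage.
; Bounces adjust gravity against pawns or boost velocity off NoStick components, while capsule overlaps with pawns fake bounces.
; Explosion halts movement.
; When restarting is enabled, stopping binds to the hit building's death or vehicle's destruction, then replicated bSimulationRunning reattaches the capsule to projectile movement via OnRep.
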(event EventBeginPlay
  ;[1 1 1] BeginPlay - Start Fuse
  (Utilities|Time|SetTimerbyEvent (EventDispatchers|CreateEvent :!event FuseEnded) (Variables|Default|GetFuseTime)))

(event EventOnBounce
  (params (Hit :type Hit%20Result%20Structure%20%28by%20ref%29))
  (bind (_bblockinghit _binitialoverlap _time _distance _location _impactpoint _normal _impactnormal _physmat _hitactor _hitcomponent) (Collision|BreakHitResult Hit))
  ;[1 1 1] OnBounce - Adjust velocity and gravity scale
  (bind _aspawn (Utilities|Casting|CastToPawn _hitactor)
    (:then
      (if (Utilities|NotEqual(Object) _aspawn (Game|GetInstigator))
        (Variables|/Script/Engine.ProjectileMovementComponent|SetProjectileGravityScale (Variables|Default|GetBouncePawnAgainstPawnGravityScale) (Variables|Movement|GetProjectileMovementComponent)) ; Hit a pawn, scale gravity
      ))
    (:CastFailed
      (if (Components|ComponentHasTag _hitcomponent "NoStick")
        (CallFunction|SetBounceVelocity)))))

(event OnComponentBeginOverlap(CapsuleComponent)
  (params (OverlappedComponent :type Primitive%20Component%20Object%20Reference) (OtherActor :type Actor%20Object%20Reference) (OtherComp :type Primitive%20Component%20Object%20Reference) (OtherBodyIndex :type Integer) (bFromSweep :type Boolean) (SweepResult :type Hit%20Result%20Structure%20%28by%20ref%29))
  (bind (_bblockinghit _binitialoverlap _time _distance _location _impactpoint _normal _impactnormal _physmat _hitactor) (Collision|BreakHitResult SweepResult))
  ;[1 1 1] If we overlap a pawn, fake a bounce.
  (bind _aspawn (Utilities|Casting|CastToPawn _hitactor))
  (if (Utilities|NotEqual(Object) _aspawn (Game|GetInstigator))
    (Variables|/Script/Engine.ProjectileMovementComponent|SetProjectileGravityScale (Variables|Default|GetBouncePawnAgainstPawnGravityScale) (Variables|Movement|GetProjectileMovementComponent)) ; Hit a pawn, scale gravity
    (CallFunction|SetBounceVelocityPawn)))

(event EventOnExploded
  (params (HitActors :type Array%20of%20Actor%20Object%20References) (HitResults :type Array%20of%20Hit%20Result%20Structures))
  ;[1 1 1] OnExploded - Stop Simulating
  (Components|Movement|SetUpdatedComponent (Variables|Movement|GetProjectileMovementComponent) 0))

(event Game|Damage|EventAnyDamage
  (params (Damage :type Float%20%28single-precision%29) (DamageType :type Damage%20Type%20Object%20Reference) (InstigatedBy :type Controller%20Object%20Reference) (DamageCauser :type Actor%20Object%20Reference))
  ;[1 1 1] AnyDamage - explode
  (Utilities|FlowControl|Delay 0.1)
  (Delivery|DoExplosion))

(event OnProjectileStop(ProjectileMovementComponent)
  (params (ImpactResult :type Hit%20Result%20Structure%20%28by%20ref%29))
  (bind (_bblockinghit _binitialoverlap _time _distance _location _impactpoint _normal _impactnormal _physmat _hitactor) (Collision|BreakHitResult ImpactResult))
  ;[1 1 1] Bind events for resimulating after the attached building actor or vehicle is destroyed
  (if (Variables|Default|GetbShouldRestartSimulating)
    (Variables|Default|SetbSimulationRunning false)
    (bind _asbuilding_actor (Utilities|Casting|CastToBuildingActor _hitactor)
      (:then
        (EventDispatchers|BindEventtoOnDied _asbuilding_actor (delegate Custom|Event_BuildingActorDied)))
      (:CastFailed
        (bind _asfort_vehicle_interface (Utilities|Casting|CastToFortVehicleInterface _hitactor))
        (Game|BindEventtoOnDestroyed (Vehicle|GetActor _asfort_vehicle_interface) (delegate Custom|Event_OnVehicleDestroyed))))))

(event Custom|RestartSimulating
  ;[1 1 1] Attached target destroyed. Start simulating again.
  (Variables|Default|SetbSimulationRunning true)
  (Networking|ForceNetUpdate))

(event Custom|Event_BuildingActorDied
  (params (DamagedActor :type Actor%20Object%20Reference) (Damage :type Float%20%28single-precision%29) (InstigatedBy :type Controller%20Object%20Reference) (DamageCauser :type Actor%20Object%20Reference) (HitLocation :type Vector) (FHitComponent :type Primitive%20Component%20Object%20Reference) (BoneName :type Name) (Momentum :type Vector))
  ;[1 1 1] Bind events for resimulating after the attached building actor or vehicle is destroyed
  (CallFunction|RestartSimulating))

(event Custom|Event_OnVehicleDestroyed
  (params (DestroyedActor :type Actor%20Object%20Reference))
  ;[1 1 1] Bind events for resimulating after the attached building actor or vehicle is destroyed
  (CallFunction|RestartSimulating))

(fn FuseEnded)

(fn OnRep_bSimulationRunning
  (if (Variables|Default|GetbSimulationRunning)
    (CallFunction|ResetSimulatedComponents)))

(fn ResetSimulatedComponents
  (Components|Movement|SetUpdatedComponent (Variables|Movement|GetProjectileMovementComponent) (Variables|Collision|GetCapsuleComponent)))

(fn SetBounceVelocity
  (Variables|/Script/Engine.MovementComponent|SetVelocity (Math|Vector|vector+vector (* (Variables|/Script/Engine.MovementComponent|GetVelocity (Variables|Movement|GetProjectileMovementComponent)) 2.75) (Math|Vector|MakeVector 0.0 0.0 (* (Math|Vector|VectorLength (Variables|/Script/Engine.MovementComponent|GetVelocity (Variables|Movement|GetProjectileMovementComponent))) 0.8))) (Variables|Movement|GetProjectileMovementComponent)))

(fn SetBounceVelocityPawn
  (Variables|/Script/Engine.MovementComponent|SetVelocity (* (Math|Vector|MakeVector (.x (* (Variables|/Script/Engine.MovementComponent|GetVelocity (Variables|Movement|GetProjectileMovementComponent)) -1)) (.y (* (Variables|/Script/Engine.MovementComponent|GetVelocity (Variables|Movement|GetProjectileMovementComponent)) -1)) (Math|Float|Absolute(Float) (.z (* (Variables|/Script/Engine.MovementComponent|GetVelocity (Variables|Movement|GetProjectileMovementComponent)) -1)))) (Variables|/Script/Engine.ProjectileMovementComponent|GetBounciness (Variables|Movement|GetProjectileMovementComponent))) (Variables|Movement|GetProjectileMovementComponent)))

(fn ConstructionScript)
